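; BP_Commerce_TieredShelf_A is a commerce tiered shelf prop Blueprint built around an instanced static mesh component.
; Its construction script calls the parent, runs a construct step on that ISM component, then sets its collision to Query and Physics.
; BeginPlay, overlap and Tick events are present but empty, so all of the graph's work happens at construction time.
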(event EventBeginPlay)

(event Collision|EventActorBeginOverlap
  (params (OtherActor :type Actor%20Object%20Reference)))

(event EventTick
  (params (DeltaSeconds :type Float%20%28single-precision%29)))

(fn ConstructionScript
  (|CallParentFunction)
  (Construct|RunConstruct (Variables|Default|GetISM_Commerce_TieredShelf_A))
  (Collision|SetCollisionEnabled (Variables|Default|GetISM_Commerce_TieredShelf_A) "QueryAndPhysics"))
